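# Niagara system for shopping cart vehicle wheel dust, with smoke sprites and sharp mesh puffs.
# User Left/Right Slow and Fast parameters scale per-side spawn rates, and VFast stretches the sharp meshes.
# Spawn groups double as bone/socket indices on the user skeletal mesh, placing particles and mirroring sideways velocity per side.
# Older skeleton-driven sparks, debris, smoke and speed-line emitters remain but are disabled, as is LeftChunks.
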NiagaraSystem /Game/Athena/DrivableVehicles/FX/NS_ShoppingCart/NS_Shopping_Cart_Vehicle_WheelDust
dependencies {
 usedDataInterfaces [
  &NiagaraDataInterfaceCurve
  &NiagaraDataInterfaceVector2DCurve
  &NiagaraDataInterfaceCurlNoise
  &NiagaraDataInterfaceVectorCurve
  &NiagaraDataInterfaceMeshRendererInfo
  &NiagaraDataInterfaceArrayFloat3
  &NiagaraDataInterfaceColorCurve
 ]
 usedDynamicInputs []
 usedModules [
  &self:SystemSpawnScript.SystemScriptSource.SystemScriptGraph.NiagaraNodeAssignment_0.SetVariables_#1
  &SystemState
  /Game/Effects/Niagara/Emitters/Vehicles/NM_DeltaOwnerXAxis
  /Game/Effects/Niagara/Modules/SkeletalMesh/FN_SkeletalBoneData
 ]
}
emitters {
 FN_Emitter {
  bEnabled=false
  emitterUpdateScript [
   &EmitterScalability(
    Module.bApplySpawnCountScale=false
    Module.bApplySpawnCountScaleByDistanceFraction=false
    Module.bUseMaxDistance=false
    Module.bUseMinDistance=false
   )
   &EmitterLifeCycle()
   &SpawnRate(SpawnRate=2, Use Spawn Probability=false)
  ]
  particleSpawnScript [
   /Niagara/Modules/Spawn/Initialization/InitializeParticle(
    Module.Write Color=true
    Module.Write Lifetime=true
    Module.Write Mass=true
    Module.Write Material Random=true
    Module.Write Position=false
    Module.Write Scale=false
    Module.Write SpriteRotation=false
    Module.Write SpriteSize=true
    Position=&SimulationPosition()
    Sprite Size=<100 100>
   )
   /SetVariables(Particles.Position=&Add_Vector(A=<0 0 0>, B=<0 0 200>)) {isSetParameters=true}
  ]
  particleUpdateScript [
   &ParticleState()
   &DynamicMaterialParameters(
    Index 0 Param 1=$System.VelocityBias
    Module.Param0WriteEnabled=true
    Module.Param1WriteEnabled=false
    Module.Param2WriteEnabled=false
    Module.Param3WriteEnabled=false
   )
  ]
  rendererClasses [&NiagaraSpriteRendererProperties]
  renderers [{rendererClass=&NiagaraSpriteRendererProperties}]
 }
 FN_MeshFromSkeleton_Sharp {
  bEnabled=false
  emitterUpdateScript [
   &EmitterScalability(
    MaxDistance=1200
    Module.bApplySpawnCountScale=false
    Module.bApplySpawnCountScaleByDistanceFraction=false
    Module.bUseMaxDistance=true
    Module.bUseMinDistance=false
    SpawnCountScaleByDistanceFraction=DataInterface() {propertyValues {Curve {keys [<at 0 1>,<at 0.5 1>,<at 1 0>]}}}
   )
   &EmitterLifeCycle()
   /SetVariables(Emitter.SpawnRateCurve=DataInterface() {propertyValues {Curve {keys [<at 0.0066 0>,<at 0.38 0>,<at 0.4 37>,<at 1 64>]}}}) {isSetParameters=true}
   /Game/Effects/Niagara/Emitters/Vehicles/SpeedTriggerCurve(MaxMovingRate:NiagaraFloat=$System.MaxMovingRate, SpawnRateCurve:NiagaraDataInterfaceCurve=$Emitter.SpawnRateCurve)
   &SpawnRate(SpawnRate=$Emitter.ResampledSpeedValue, Use Spawn Probability=false)
   &SpawnBurst_Instantaneous(Spawn Count=5, Use Spawn Probability=false) {enabled=false}
  ]
  particleSpawnScript [
   /SetVariables(Particles.SpawnGroup=HlslExpression() {hlslExpression="int(fmod(float(Particles.UniqueID), 1.0))"}) {enabled=false, isSetParameters=true}
   /Game/Effects/Niagara/Modules/Groups/FN_SetSpawnGroup_v1/FN_SetSpawnGroup(Emitter Bone Count=2)
   /Game/Effects/Niagara/Modules/SkeletalMesh/FN_UpdateSpawnGroupsToSystemSkeletalMeshSample(
    Vector 0:Vector3f=$System.SkeletalBoneData002.Position Local
    Vector 1:Vector3f=$System.SkeletalBoneData003.Position Local
    Vector 2:Vector3f=$System.SkeletalBoneData002.Position Local
    Vector 3:Vector3f=$System.SkeletalBoneData003.Position Local
   )
   &InitializeParticle(
    Color=<rgba 0.1667 0.1381 0.1224 1>
    Lifetime=&UniformRangedFloat(Minimum=0.05, Maximum=0.2)
    Mesh Scale=&UniformRangedVector(Minimum=<0.4 0.3 0.3>, Maximum=<0.7 0.5 0.5>)
    Module.Write Color=true
    Module.Write Lifetime=true
    Module.Write Mass=true
    Module.Write Material Random=true
    Module.Write Position=false
    Module.Write Scale=true
    Module.Write SpriteRotation=false
    Module.Write SpriteSize=false
    Position=&SimulationPosition()
   )
   /Niagara/Modules/Spawn/Orientation/V2/InitialMeshOrientation(
    Orientation Vector=<1 0.45 0>
    Rotation=&UniformRangedVector(Minimum=<-0.1 -0.1 -0.1>, Maximum=<0.1 0.1 0.1>)
    Rotation Coordinate Space:ENiagaraMeshTransforms=::NewEnumerator2
    Use Orientation Vector=false
    Use Rotation Vector=true
   )
   /SetVariables(Particles.Position=&Add_Vector(A=Linked() {linkedVariable {name=Particles.Position, type=NiagaraPosition}}, B=<-5 0 0>)) {enabled=false, isSetParameters=true}
   &AddVelocity(Velocity=<-10 0 0>)
  ]
  particleUpdateScript [
   &ParticleState()
   &SampleSkeletalMeshSkeleton() {enabled=false}
   /Game/Effects/Niagara/Modules/SkeletalMesh/FN_LockLastParticleToBone()
   &SolveForcesAndVelocity(Clamp Velocity=false, Limit Acceleration=false)
   &Color() {enabled=false}
   &CameraOffset(Camera Offset Amount=50000)
   &ScaleColor(
    Scale Alpha=&FloatFromCurve(FloatCurve=DataInterface() {propertyValues {Curve {keys [
         <at -0.0046 0 RCIM_Cubic>
         {arriveTangent=0.0019, interpMode=RCIM_Cubic, leaveTangent=0.0019, time=0.14, value=1}
         <at 1.0011 0.002 RCIM_Cubic>
        ]}}})
    ScaleA=true
    ScaleRGB=true
   )
   &ScaleMeshSize(Scale Factor=&VectorFromCurve(VectorCurve=DataInterface() {propertyValues {
       XCurve {keys [<at -0.0011 0.5 RCIM_Cubic>,<at 0.9989 1.5 RCIM_Cubic>]}
       YCurve {keys [<at -0.0011 0.5 RCIM_Cubic>,<at 0.9989 1.5 RCIM_Cubic>]}
       ZCurve {keys [<at -0.0011 0.5 RCIM_Cubic>,<at 0.9989 1.5 RCIM_Cubic>]}
      }}))
  ]
  rendererClasses [&NiagaraMeshRendererProperties]
  renderers [{rendererClass=&NiagaraMeshRendererProperties, rendererIndex=0}]
 }
 FN_SpriteFromSkeleton_Debirs {
  bEnabled=false
  emitterUpdateScript [
   &EmitterScalability(
    MaxDistance=500
    Module.bApplySpawnCountScale=false
    Module.bApplySpawnCountScaleByDistanceFraction=false
    Module.bUseMaxDistance=true
    Module.bUseMinDistance=false
   )
   &EmitterLifeCycle()
   &SpawnPerUnit(
    Module.Use Max Movement Threshold=true
    Module.Use Movement Tolerance=true
    Movement Tolerance=10
    Spawn Per Unit=40
    Use Spawn Probability=false
   )
   &SpawnBurst_Instantaneous(Spawn Count=5, Use Spawn Probability=false) {enabled=false}
  ]
  particleSpawnScript [
   &FN_SetSpawnGroup(Emitter Bone Count=4)
   &FN_UpdateSpawnGroupsToSystemSkeletalMeshSample(
    Vector 0:Vector3f=$System.SkeletalBoneData.Position
    Vector 1:Vector3f=$System.SkeletalBoneData001.Position
    Vector 2:Vector3f=$System.SkeletalBoneData002.Position
    Vector 3:Vector3f=$System.SkeletalBoneData003.Position
   )
   &InitializeParticle(
    Color=<rgba 0.0625 0.0625 0.0625 1>
    Lifetime=&UniformRangedFloat(Minimum=0.2, Maximum=0.4)
    Module.Write Color=true
    Module.Write Lifetime=true
    Module.Write Mass=true
    Module.Write Material Random=true
    Module.Write Position=false
    Module.Write Scale=false
    Module.Write SpriteRotation=true
    Module.Write SpriteSize=true
    Position=&SimulationPosition()
    Sprite Rotation=&UniformRangedFloat(Minimum=0, Maximum=360)
    Sprite Size=&UniformRangedVector2D(Minimum=<3 12>, Maximum=<2 8>)
   )
   &AddVelocity(Velocity=&UniformRangedVector(Minimum=<-100 -100 200>, Maximum=<100 100 400>))
   &InheritVelocity(Module.Limit Speed=false)
  ]
  particleUpdateScript [
   &ParticleState()
   &SampleSkeletalMeshSkeleton(
    Mesh=DataInterface() {propertyValues {
      DefaultMesh=null
      ExcludeBoneName=root
      OptionalComponentClass=null
      PreviewMesh=&M_Med_Soldier_04_Preview
      SkinningMode=SkinOnTheFly
      SoftSourceActor=null
      SourceMode=Default
      UvSetIndex=0
      WholeMeshLOD=-1
      bExcludeBone=false
      bReadDeformedGeometry=true
      bRequireCurrentFrameData=true
      ComponentTags []
      FilteredBones [hand_r,hand_l,foot_r,foot_l]
      FilteredSockets []
      MeshUserParameter {parameter {name=None, typeDefHandle {registeredTypeIndex=1}, varData []}}
      SamplingRegions []
     }}
    Specific Bone Index:NiagaraInt32=$Particles.SpawnGroup
   ) {enabled=false}
   &GravityForce()
   &FN_LockLastParticleToBone(Locked Position:Vector3f=$Particles.SampleSkeletalMeshSkeleton.SampledPosition, Only Last Particle Born=false)
   &ScaleSpriteSize(Scale Factor=&Vector2DFromFloat(Value=&FloatFromCurve(FloatCurve=DataInterface() {propertyValues {Curve {keys [<at 0.5 1 RCIM_Cubic>,<at 1.0011 0 RCIM_Cubic>]}}})))
   &Drag(Drag=2, Use Linear Drag=true, Use Rotational Drag=false)
   &SolveForcesAndVelocity(Clamp Velocity=false, Limit Acceleration=false)
  ]
  rendererClasses [&NiagaraSpriteRendererProperties]
  renderers [{rendererClass=&NiagaraSpriteRendererProperties}]
 }
 FN_SpriteFromSkeleton_Smoke {
  bEnabled=false
  emitterUpdateScript [
   &EmitterScalability(
    MaxDistance=6000
    Module.bApplySpawnCountScale=false
    Module.bApplySpawnCountScaleByDistanceFraction=false
    Module.bUseMaxDistance=true
    Module.bUseMinDistance=false
    SpawnCountScaleByDistanceFraction=DataInterface() {propertyValues {Curve {keys [<at 0 1>,<at 0.2 1>,<at 1 0>]}}}
   )
   &EmitterLifeCycle()
   /SetVariables(Emitter.SpawnRateCurve=DataInterface() {propertyValues {Curve {keys [<at -0.0021 0>,<at 0.245 0>,<at 0.27 90>,<at 1 206>]}}}) {isSetParameters=true, module=SetVariables_#2}
   &SpeedTriggerCurve(MaxMovingRate:NiagaraFloat=$System.MaxMovingRate, SpawnRateCurve:NiagaraDataInterfaceCurve=$Emitter.SpawnRateCurve)
   &SpawnRate(SpawnRate=$Emitter.ResampledSpeedValue, Use Spawn Probability=false)
   &SpawnBurst_Instantaneous(Spawn Count=5, Use Spawn Probability=false) {enabled=false}
  ]
  particleSpawnScript [
   /SetVariables(Particles.Random=&UniformRangedFloat(Minimum=10, Maximum=20)) {isSetParameters=true}
   &FN_SetSpawnGroup(Emitter Bone Count=4)
   &FN_UpdateSpawnGroupsToSystemSkeletalMeshSample(
    Vector 0:Vector3f=$System.SkeletalBoneData.Position
    Vector 1:Vector3f=$System.SkeletalBoneData001.Position
    Vector 2:Vector3f=$System.SkeletalBoneData002.Position
    Vector 3:Vector3f=$System.SkeletalBoneData003.Position
   )
   &InitializeParticle(
    Color=<rgba 0.5677 0.5677 0.5677 0.2>
    Lifetime=&UniformRangedFloat(Minimum=0.15, Maximum=0.3)
    Module.Write Color=true
    Module.Write Lifetime=true
    Module.Write Mass=true
    Module.Write Material Random=true
    Module.Write Position=false
    Module.Write Scale=false
    Module.Write SpriteRotation=true
    Module.Write SpriteSize=true
    Position=&SimulationPosition()
    Sprite Rotation=&UniformRangedFloat(Minimum=0, Maximum=360)
   )
   &InheritVelocity(Module.Limit Speed=false, Velocity Scale=&VectorFromFloat(Value=0.5))
   /SetVariables(Particles.Position=&Add_Vector(
     A=Linked() {linkedVariable {name=Particles.Position, type=NiagaraPosition}}
     B=&Multiply_Vector(A=$Engine.Owner.Velocity, B=&VectorFromFloat(Value=0.025))
    )) {isSetParameters=true}
   /SetVariables(Particles.Position=&Add_Vector(
     A=Linked() {linkedVariable {name=Particles.Position, type=NiagaraPosition}}
     B=&Multiply_Vector(A=$Engine.Owner.SystemXAxis, B=&VectorFromFloat(Value=15))
    )) {enabled=false, isSetParameters=true}
  ]
  particleUpdateScript [
   &ParticleState()
   &SampleSkeletalMeshSkeleton(
    Mesh=DataInterface() {propertyValues {
      DefaultMesh=null
      ExcludeBoneName=root
      OptionalComponentClass=null
      PreviewMesh=&M_Med_Soldier_04_Preview
      SkinningMode=SkinOnTheFly
      SoftSourceActor=null
      SourceMode=Default
      UvSetIndex=0
      WholeMeshLOD=-1
      bExcludeBone=false
      bReadDeformedGeometry=true
      bRequireCurrentFrameData=true
      ComponentTags []
      FilteredBones [hand_r,hand_l,foot_r,foot_l]
      FilteredSockets []
      MeshUserParameter {parameter {name=None, typeDefHandle {registeredTypeIndex=1}, varData []}}
      SamplingRegions []
     }}
    Specific Bone Index:NiagaraInt32=$Particles.SpawnGroup
   ) {enabled=false}
   &FN_LockLastParticleToBone(Locked Position:Vector3f=$Particles.SampleSkeletalMeshSkeleton.SampledPosition)
   &ScaleSpriteSize(Scale Factor=&Vector2DFromFloat(Value=&FloatFromCurve(FloatCurve=DataInterface() {propertyValues {Curve {keys [<at 0 1.2 RCIM_Cubic>,<at 1.0011 3.48 RCIM_Cubic>]}}})))
   /SetVariables(Particles.SubImageIndex=&FloatFromCurve(FloatCurve=DataInterface() {propertyValues {Curve {keys [<at -0.0046 16>,<at 1.0011 32>]}}})) {isSetParameters=true}
   /Niagara/Modules/Update/Material/DynamicMatParams(
    Module.Param0WriteEnabled=true
    Module.Param1WriteEnabled=false
    Module.Param2WriteEnabled=false
    Module.Param3WriteEnabled=false
    Param0Value:NiagaraFloat=&FloatFromCurve(FloatCurve=DataInterface() {propertyValues {Curve {keys [<at 0.38 1.0113 RCIM_Cubic>,<at 1.0011 0.002 RCIM_Cubic>]}}})
    Param3Value:NiagaraFloat=&FloatFromCurve(FloatCurve=DataInterface() {propertyValues {Curve {keys [<at -0.0046 0.2 RCIM_Cubic>,<at 1.0011 1 RCIM_Cubic>]}}})
   )
   &Drag(Drag=2, Use Linear Drag=true, Use Rotational Drag=false)
   &SolveForcesAndVelocity(Clamp Velocity=false, Limit Acceleration=false)
  ]
  rendererClasses [&NiagaraSpriteRendererProperties]
  renderers [{rendererClass=&NiagaraSpriteRendererProperties}]
 }
 FN_SpriteFromSkeleton_Sparks {
  bEnabled=false
  emitterUpdateScript [
   &EmitterScalability(
    MaxDistance=2000
    Module.bApplySpawnCountScale=false
    Module.bApplySpawnCountScaleByDistanceFraction=false
    Module.bUseMaxDistance=true
    Module.bUseMinDistance=false
    SpawnCountScaleByDistanceFraction=DataInterface() {propertyValues {Curve {keys [<at 0.3 1>,<at 1 0>]}}}
   )
   &EmitterLifeCycle(NextLoopDelay:NiagaraFloat=&UniformRangedFloat(Minimum=0, Maximum=0.1), NextLoopDuration=0.2)
   /SetVariables(Emitter.SpawnRateCurve=DataInterface() {propertyValues {Curve {keys [<at -0.0021 0>,<at 0.6051 0>,<at 0.7198 279.5417>,<at 0.9028 512>,<at 0.9991 1024>]}}}) {isSetParameters=true, module=SetVariables_#2}
   &SpeedTriggerCurve(MaxMovingRate:NiagaraFloat=$System.MaxMovingRate, SpawnRateCurve:NiagaraDataInterfaceCurve=$Emitter.SpawnRateCurve)
   &SpawnRate(
    SpawnRate=&Multiply_Float(A=$Emitter.ResampledSpeedValue, B=$System.VelocityBias)
    Use Spawn Probability=false
   ) {enabled=false}
   &SpawnBurst_Instantaneous(Spawn Count=5, Use Spawn Probability=false) {enabled=false}
  ]
  particleSpawnScript [
   /SetVariables(Particles.Random=&UniformRangedFloat(Minimum=0, Maximum=1)) {isSetParameters=true}
   &FN_SetSpawnGroup(Emitter Bone Count=4)
   &FN_UpdateSpawnGroupsToSystemSkeletalMeshSample(
    Vector 0:Vector3f=$System.SkeletalBoneData.Position
    Vector 1:Vector3f=$System.SkeletalBoneData001.Position
    Vector 2:Vector3f=$System.SkeletalBoneData002.Position
    Vector 3:Vector3f=$System.SkeletalBoneData003.Position
   )
   &InitializeParticle(
    Color=<rgba 100 43.3283 9.375 1>
    Lifetime=&FloatFromCurve(
     FloatCurve=DataInterface() {propertyValues {Curve {keys [<at 0 0.0938>,<at 0.66 0.1438>,<at 0.75 0.3>,<at 1 0.34>]}}}
     CurveIndex=$Particles.Random
    )
    Module.Write Color=true
    Module.Write Lifetime=true
    Module.Write Mass=true
    Module.Write Material Random=true
    Module.Write Position=false
    Module.Write Scale=false
    Module.Write SpriteRotation=true
    Module.Write SpriteSize=true
    Position=&SimulationPosition()
    Sprite Rotation=&UniformRangedFloat(Minimum=0, Maximum=360)
    Sprite Size=&Vector2DFromCurve(
     Vector2Curve=DataInterface() {propertyValues {
       XCurve {keys [<at 0.0007 0.218>,<at 0.5404 1.3699>,<at 0.9989 2.183>]}
       YCurve {keys [<at -0.0011 10.1707>,<at 0.5422 17.7674>,<at 0.9989 44.5076>]}
      }}
     CurveIndex=$Particles.Random
     ScaleCurve=<1 1>
    )
   )
   &AddVelocityInCone(
    Cone Axis=<-1 0 0.25>
    Module.Use Velocity Falloff On Cone Axis=false
    Velocity Strength=&Multiply_Float(
     A=&UniformRangedFloat(Minimum=0.5, Maximum=1.5)
     B=/Niagara/DynamicInputs/TypeConversions/VectorLengthSafe(Vector:Vector3f=$Engine.Owner.Velocity)
    )
   )
  ]
  particleUpdateScript [
   &ParticleState()
   &SampleSkeletalMeshSkeleton(
    Mesh:NiagaraDataInterfaceSkeletalMesh=$User.SkeletalMesh
    Spawn Mode:ENiagaraSkeletalSpawning=::NewEnumerator1
    Specific Bone Index:NiagaraInt32=$Particles.SpawnGroup
   )
   &Drag(Drag=2, Use Linear Drag=true, Use Rotational Drag=false)
   /Niagara/Modules/Update/Forces/CurlNoiseForce(Noise Strength=500, Pan Noise=false)
   &FN_LockLastParticleToBone(Locked Position:Vector3f=$Particles.SampleSkeletalMeshSkeleton.SampledPosition, Only Last Particle Born=false)
   &ScaleSpriteSize(Scale Factor=&Vector2DFromFloat(Value=&FloatFromCurve(FloatCurve=DataInterface() {propertyValues {Curve {keys [<at -0.0046 1.0113>,<at 1.0011 0.002>]}}})))
   &SolveForcesAndVelocity(Clamp Velocity=false, Limit Acceleration=false)
  ]
  rendererClasses [&NiagaraSpriteRendererProperties]
  renderers [{rendererClass=&NiagaraSpriteRendererProperties}]
 }
 FN_SpriteFromSkeleton_SparksLocal {
  bEnabled=false
  emitterUpdateScript [
   &EmitterScalability(
    MaxDistance=500
    Module.bApplySpawnCountScale=false
    Module.bApplySpawnCountScaleByDistanceFraction=false
    Module.bUseMaxDistance=true
    Module.bUseMinDistance=false
   )
   &EmitterLifeCycle()
   /SetVariables(Emitter.SpawnRateCurve=DataInterface() {propertyValues {Curve {keys [<at -0.0021 0>,<at 0.477 0>,<at 0.78 512>,<at 0.9281 1024>,<at 1.0022 2048>]}}}) {isSetParameters=true, module=SetVariables_#2}
   &SpeedTriggerCurve(MaxMovingRate:NiagaraFloat=$System.MaxMovingRate, SpawnRateCurve:NiagaraDataInterfaceCurve=$Emitter.SpawnRateCurve)
   &SpawnRate(
    SpawnRate=&Multiply_Float(A=$Emitter.ResampledSpeedValue, B=$System.VelocityBias)
    Use Spawn Probability=false
   )
   &SpawnBurst_Instantaneous(Spawn Count=5, Use Spawn Probability=false) {enabled=false}
  ]
  particleSpawnScript [
   /SetVariables(Particles.Random=&UniformRangedFloat(Minimum=10, Maximum=20)) {isSetParameters=true}
   &FN_SetSpawnGroup(Emitter Bone Count=4)
   &FN_UpdateSpawnGroupsToSystemSkeletalMeshSample(
    Vector 0:Vector3f=$System.SkeletalBoneData.Position Local
    Vector 1:Vector3f=$System.SkeletalBoneData001.Position Local
    Vector 2:Vector3f=$System.SkeletalBoneData002.Position Local
    Vector 3:Vector3f=$System.SkeletalBoneData003.Position Local
   )
   &InitializeParticle(
    Color=<rgba 100 43.3283 9.375 1>
    Lifetime=&FloatFromCurve(FloatCurve=DataInterface() {propertyValues {Curve {keys [<at 0 0.07>,<at 1 0.117>]}}}, CurveIndex=$Particles.Random)
    Module.Write Color=true
    Module.Write Lifetime=true
    Module.Write Mass=true
    Module.Write Material Random=true
    Module.Write Position=false
    Module.Write Scale=false
    Module.Write SpriteRotation=true
    Module.Write SpriteSize=true
    Position=&SimulationPosition()
    Sprite Rotation=&UniformRangedFloat(Minimum=0, Maximum=360)
    Sprite Size=&Vector2DFromCurve(
     Vector2Curve=DataInterface() {propertyValues {
       XCurve {keys [<at -0.0039 0.9112>,<at 0.5404 1.9285>,<at 0.9989 2.5081>]}
       YCurve {keys [<at -0.0011 19.4182>,<at 0.5422 38.4149>,<at 0.9989 51.4852>]}
      }}
     CurveIndex=$Particles.Random
     ScaleCurve=<1 1>
    )
   )
   &AddVelocityInCone(
    Cone Axis=<-1 0 0.25>
    Module.Use Velocity Falloff On Cone Axis=false
    Velocity Strength=&UniformRangedFloat(Minimum=400, Maximum=700)
   )
   /SetVariables(Particles.Position=HlslExpression() {hlslExpression="Particles.Position + (Engine.Owner.SystemZAxis * -5.0f)"}) {isSetParameters=true}
  ]
  particleUpdateScript [
   &ParticleState()
   &SampleSkeletalMeshSkeleton(
    Mesh=DataInterface() {propertyValues {
      DefaultMesh=null
      ExcludeBoneName=root
      OptionalComponentClass=null
      PreviewMesh=&M_Med_Soldier_04_Preview
      SkinningMode=SkinOnTheFly
      SoftSourceActor=null
      SourceMode=Default
      UvSetIndex=0
      WholeMeshLOD=-1
      bExcludeBone=false
      bReadDeformedGeometry=true
      bRequireCurrentFrameData=true
      ComponentTags []
      FilteredBones [hand_r,hand_l,foot_r,foot_l]
      FilteredSockets []
      MeshUserParameter {parameter {name=None, typeDefHandle {registeredTypeIndex=1}, varData []}}
      SamplingRegions []
     }}
    Specific Bone Index:NiagaraInt32=$Particles.SpawnGroup
   )
   &Drag(Drag=2, Use Linear Drag=true, Use Rotational Drag=false)
   &CurlNoiseForce(Noise Strength=500, Pan Noise=false)
   &FN_LockLastParticleToBone(Locked Position:Vector3f=$Particles.SampleSkeletalMeshSkeleton.SampledPosition, Only Last Particle Born=false)
   &ScaleSpriteSize(Scale Factor=&Vector2DFromFloat(Value=&FloatFromCurve(FloatCurve=DataInterface() {propertyValues {Curve {keys [<at -0.0046 1.0113>,<at 1.0011 0.002>]}}})))
   &SolveForcesAndVelocity(Clamp Velocity=false, Limit Acceleration=false)
  ]
  rendererClasses [&NiagaraSpriteRendererProperties]
  renderers [{rendererClass=&NiagaraSpriteRendererProperties}]
 }
 FN_Sprite_SpeedLines {
  bEnabled=false
  emitterUpdateScript [
   &SpawnBurst_Instantaneous(Spawn Count=5, Use Spawn Probability=false) {enabled=false}
   /SetVariables(Emitter.SpawnRateCurve=DataInterface() {propertyValues {Curve {keys [
        {arriveTangent=355.9891, interpMode=RCIM_Cubic, leaveTangent=355.9867, tangentMode=RCTM_User, time=0.35, value=0}
        <at 0.9981 250 RCIM_Cubic>
       ]}}}) {isSetParameters=true, module=SetVariables_#3}
   &SpeedTriggerCurve(
    MaxMovingRate:NiagaraFloat=&Multiply_Float(A=$System.MaxMovingRate, B=2)
    SpawnRateCurve:NiagaraDataInterfaceCurve=$Emitter.SpawnRateCurve
   )
   &EmitterState()
   &SpawnRate(SpawnRate=$Emitter.ResampledSpeedValue, Use Spawn Probability=false)
  ]
  particleSpawnScript [
   /Niagara/Modules/Spawn/Initialization/V2/InitializeParticle(
    Color=<rgba 0.2292 0.1989 0.1838 1>
    Lifetime Max=0.6
    Lifetime Min=0.4
    Lifetime Mode=::NewEnumerator1
    Sprite Rotation Mode=::NewEnumerator2
    Sprite Size Max=<5 512>
    Sprite Size Min=<3 240>
    Sprite Size Mode=::NewEnumerator2
    UsePositionOffset=false
    Write Material Random=true
   ) {module=InitializeParticle001}
   /Niagara/Modules/Spawn/Initialization/InitializeParticle(
    Color=&MakeLinearColor(
     R:NiagaraFloat=0.2292
     G:NiagaraFloat=0.1989
     B:NiagaraFloat=0.1838
     A:NiagaraFloat=&UniformRangedFloat(Minimum=0.2, Maximum=0.4)
    )
    Lifetime=&UniformRangedFloat(Minimum=0.4, Maximum=0.6)
    Module.Write Color=true
    Module.Write Lifetime=true
    Module.Write Mass=true
    Module.Write Material Random=true
    Module.Write Position=false
    Module.Write Scale=false
    Module.Write SpriteRotation=false
    Module.Write SpriteSize=true
    Position=&SimulationPosition()
    Sprite Rotation=&UniformRangedFloat(Minimum=0, Maximum=360)
    Sprite Size=&UniformRangedVector2D(Minimum=<3.0297 240>, Maximum=<5 512>)
   )
   &TorusLocation(
    Align Torus With Axis=true
    Handle Radius=32
    Large Radius:NiagaraFloat=&UniformAOrBFloat(A=180, B=256)
    MaskSpawnGroup=false
    Module.Override Randomness=false
    Offset=&Multiply_VectorByFloat(Vector=$Engine.Owner.SystemXAxis, Float=40)
    Torus Axis=<1 0 0>
   )
   /SetVariables(Particles.Position=&Add_Vector(
     A=Linked() {linkedVariable {name=Particles.Position, type=NiagaraPosition}}
     B=&Multiply_Vector(A=$Engine.Owner.Velocity, B=&VectorFromFloat(Value=0.15))
    )) {isSetParameters=true, module=SetVariables_#4}
   &AddVelocity(Velocity=&MakeVector(X=&UniformRangedFloat(Minimum=-50, Maximum=-100), Y=0, Z=0))
   &InheritVelocity(Module.Limit Speed=false, Velocity Scale=&VectorFromFloat(Value=-0.5))
  ]
  particleUpdateScript [
   &ParticleState()
   /Game/Effects/Niagara/Emitters/Water/FN_DistanceKillParticles(AgeReduiceAmount=0.0001, CenterPosition:Vector3f=&SimulationPosition()) {enabled=false}
   &Color(
    Color=<rgba 1 1 1 0.1>
    Scale Alpha=&FloatFromCurve(
     FloatCurve=DataInterface() {propertyValues {Curve {keys [<at 0 0 RCIM_Cubic>,<at 0.5 1 RCIM_Cubic>,<at 1 0 RCIM_Cubic>]}}}
     Scale Curve=&RandomRangeFloat(Minimum=0.2, Maximum=0.4)
    )
   )
   &SolveForcesAndVelocity(Clamp Velocity=false, Limit Acceleration=false)
   &ScaleSpriteSizeBySpeed(Max Scale Factor=<5 10>, Min Scale Factor=<0 0.1>)
   &ScaleSpriteSize(Scale Factor=&Vector2DFromCurve(
     Vector2Curve=DataInterface() {propertyValues {
       XCurve {keys [<at -0.0069 0.5 RCIM_Cubic>,<at 0.4918 1 RCIM_Cubic>,<at 1 0.5 RCIM_Cubic>]}
       YCurve {keys [<at 0 0.0245 RCIM_Cubic>, {arriveTangent=-0.0245, interpMode=RCIM_Cubic, leaveTangent=-0.0245, time=0.4975, value=1}, <at 1 0 RCIM_Cubic>]}
      }}
     ScaleCurve=<1 1>
    ))
   &CameraOffset(Camera Offset Amount=-100)
  ]
  rendererClasses [&NiagaraSpriteRendererProperties]
  renderers [{rendererClass=&NiagaraSpriteRendererProperties}]
 }
 FN_Sprite_SpeedLinesSlow {
  bEnabled=false
  emitterUpdateScript [
   &SpawnBurst_Instantaneous(Spawn Count=5, Use Spawn Probability=false) {enabled=false}
   /SetVariables(Emitter.SpawnRateCurve=DataInterface() {propertyValues {Curve {keys [{arriveTangent=141.7202, interpMode=RCIM_Cubic, leaveTangent=141.719, tangentMode=RCTM_User, time=0.35, value=0}, <at 0.9981 100 RCIM_Cubic>]}}}) {isSetParameters=true, module=SetVariables_#3}
   &SpeedTriggerCurve(
    MaxMovingRate:NiagaraFloat=&Multiply_Float(A=$System.MaxMovingRate, B=2)
    SpawnRateCurve:NiagaraDataInterfaceCurve=$Emitter.SpawnRateCurve
   )
   &EmitterState()
   &SpawnRate(SpawnRate=$Emitter.ResampledSpeedValue, Use Spawn Probability=false)
  ]
  particleSpawnScript [
   &InitializeParticle(
    Color=&MakeLinearColor(
     R:NiagaraFloat=0.2292
     G:NiagaraFloat=0.1989
     B:NiagaraFloat=0.1838
     A:NiagaraFloat=&UniformRangedFloat(Minimum=0.2, Maximum=0.4)
    )
    Lifetime=&UniformRangedFloat(Minimum=0.8, Maximum=0.4)
    Module.Write Color=true
    Module.Write Lifetime=true
    Module.Write Mass=true
    Module.Write Material Random=true
    Module.Write Position=false
    Module.Write Scale=false
    Module.Write SpriteRotation=false
    Module.Write SpriteSize=true
    Position=&SimulationPosition()
    Sprite Rotation=&UniformRangedFloat(Minimum=0, Maximum=360)
    Sprite Size=&UniformRangedVector2D(Minimum=<5 100>, Maximum=<8 200>)
   )
   &TorusLocation(
    Align Torus With Axis=true
    Handle Radius=32
    Large Radius:NiagaraFloat=&UniformAOrBFloat(A=30, B=40)
    MaskSpawnGroup=false
    Module.Override Randomness=false
    Offset=&Multiply_VectorByFloat(Vector=$Engine.Owner.SystemXAxis, Float=40)
    Torus Axis=<1 0 0>
   )
   /SetVariables(Particles.Position=HlslExpression() {hlslExpression="Particles.Position + (Engine.Owner.Velocity * 0.15f) + (Engine.Owner.SystemZAxis * 80.0f)"}) {isSetParameters=true, module=SetVariables_#4}
   &AddVelocity(Velocity=&MakeVector(X=&UniformRangedFloat(Minimum=-50, Maximum=-100), Y=0, Z=0))
   &InheritVelocity(Module.Limit Speed=false, Velocity Scale=&VectorFromFloat(Value=0.2))
  ]
  particleUpdateScript [
   &ParticleState()
   &FN_DistanceKillParticles(AgeReduiceAmount=0.0001, CenterPosition:Vector3f=&SimulationPosition()) {enabled=false}
   &Color(
    Color=<rgba 1 1 1 0.1>
    Scale Alpha=&FloatFromCurve(FloatCurve=DataInterface() {propertyValues {Curve {keys [<at 0 0 RCIM_Cubic>,<at 0.5 1 RCIM_Cubic>,<at 1 0 RCIM_Cubic>]}}})
   )
   &SolveForcesAndVelocity(Clamp Velocity=false, Limit Acceleration=false)
   &ScaleSpriteSizeBySpeed(Max Scale Factor=<5 10>, Min Scale Factor=<0 0.1>)
   &ScaleSpriteSize(Scale Factor=&Vector2DFromCurve(
     Vector2Curve=DataInterface() {propertyValues {
       XCurve {keys [<at -0.0069 0.5 RCIM_Cubic>,<at 0.4918 1 RCIM_Cubic>,<at 1 0.5 RCIM_Cubic>]}
       YCurve {keys [<at 0 0.0245 RCIM_Cubic>, {arriveTangent=-0.0245, interpMode=RCIM_Cubic, leaveTangent=-0.0245, time=0.4975, value=1}, <at 1 0 RCIM_Cubic>]}
      }}
     ScaleCurve=<1 1>
    ))
   &CameraOffset(Camera Offset Amount=-100)
  ]
  rendererClasses [&NiagaraSpriteRendererProperties]
  renderers [{rendererClass=&NiagaraSpriteRendererProperties}]
 }
 LeftChunks {
  bEnabled=false
  emitterUpdateScript [&EmitterState(
    Life Cycle Mode=::NewEnumerator1
    Loop Duration=&RandomRangeFloat(Minimum=0.5, Maximum=1.5, Evaluation Type=::NewEnumerator1)
    Recalculate Duration Each Loop=true
    UseLoopDelay=true
   )]
  particleSpawnScript [
   /SetVariables(Particles.SpriteRotationRate=&RandomRangeFloat(Minimum=-6, Maximum=6, Evaluation Type=::NewEnumerator1)) {isSetParameters=true}
   /Niagara/Modules/Spawn/Initialization/V2/InitializeParticle(UsePositionOffset=false, Write Material Random=true) {module=InitializeParticle001}
   /Niagara/Modules/Spawn/Initialization/InitializeParticle(
    Lifetime=&RandomRangeFloat(Minimum=0.4, Maximum=0.7, Evaluation Type=::NewEnumerator1)
    Module.Write Color=true
    Module.Write Lifetime=true
    Module.Write Mass=true
    Module.Write Material Random=true
    Module.Write Position=false
    Module.Write Scale=false
    Module.Write SpriteRotation=true
    Module.Write SpriteSize=true
    Position=<0 0 -6>
    Sprite Rotation=&NormalizedAngleToDegrees(NormalizedAngle=&RandomRangeFloat(Minimum=0, Maximum=1, Evaluation Type=::NewEnumerator1))
    Sprite Size=&RandomRangeVector2D(Minimum=<2 4>, Maximum=<4 14>, Evaluation Type=::NewEnumerator1)
   )
   &Color(Color=&MakeLinearColorFromVectorAndFloat("Vector (RGB)"=<0.1042 0.1042 0.1042>, "Float (Alpha)"=&ClampFloat(Float=2)))
   &AddVelocity(Velocity=&RandomRangeVector(Minimum=<-20 -150 150>, Maximum=<-150 150 500>, Evaluation Type=::NewEnumerator1))
   &AddVelocityFromPoint(
    Module.Use Falloff Distance=false
    Velocity Strength=&RandomRangeFloat(Minimum=0, Maximum=300, Evaluation Type=::NewEnumerator1)
   )
   &DynamicMaterialParameters(
    Index 0 Param 1=&RandomRangeFloat(Minimum=0, Maximum=1, Evaluation Type=::NewEnumerator1)
    Module.Param0WriteEnabled=true
    Module.Param1WriteEnabled=false
    Module.Param2WriteEnabled=false
    Module.Param3WriteEnabled=false
   )
   &SkeletalMeshLocation(
    Bone Sampling Mode=::NewEnumerator4
    MaskSpawnGroup=false
    Skeletal Mesh=DataInterface() {propertyValues {
      DefaultMesh=null
      ExcludeBoneName=root
      OptionalComponentClass=null
      PreviewMesh=null
      SkinningMode=SkinOnTheFly
      SoftSourceActor=null
      SourceMode=Default
      UvSetIndex=0
      WholeMeshLOD=-1
      bExcludeBone=false
      bReadDeformedGeometry=true
      bRequireCurrentFrameData=true
      ComponentTags []
      FilteredBones [wheel_BL]
      FilteredSockets []
      MeshUserParameter {parameter {name=None, typeDefHandle {registeredTypeIndex=1}, varData []}}
      SamplingRegions []
     }}
   )
  ]
  particleUpdateScript [
   &ParticleState()
   &AccelerationForce(Acceleration=<0 0 -1600>)
   &SolveForcesAndVelocity(Clamp Velocity=false, Limit Acceleration=false)
   &ScaleSpriteSize()
   &SpriteRotationRate(Rotation Rate=&NormalizedAngleToDegrees(NormalizedAngle=$Particles.SpriteRotationRate))
   &DynamicMaterialParameters(
    Index 0 Param 2=1
    Index 0 Param 3=1
    Index 0 Param 4=1
    Module.Param0WriteEnabled=false
    Module.Param1WriteEnabled=true
    Module.Param2WriteEnabled=true
    Module.Param3WriteEnabled=true
   ) {module=DynamicMaterialParameters001}
  ]
  rendererClasses [&NiagaraSpriteRendererProperties]
  renderers [{rendererClass=&NiagaraSpriteRendererProperties}]
 }
 Sharp {
  emitterUpdateScript [
   &EmitterState(
    Life Cycle Mode=::NewEnumerator1
    Loop Duration=&RandomRangeFloat(Minimum=0.5, Maximum=1.2, Evaluation Type=::NewEnumerator1)
    Recalculate Duration Each Loop=true
    UseLoopDelay=true
   )
   &SpawnRate(SpawnGroup=2, SpawnRate=&Lerp_Float(A=0, B=8, Alpha=$User.LeftFast), Use Spawn Probability=false)
   &SpawnRate(SpawnGroup=3, SpawnRate=&Lerp_Float(A=0, B=8, Alpha=$User.RightFast), Use Spawn Probability=false) {module=SpawnRate001}
  ]
  particleSpawnScript [
   /Niagara/Modules/Spawn/Initialization/V2/InitializeParticle(
    Color=<rgba 0.37 0.2887 0.2698 1>
    Lifetime Max=0.2
    Lifetime Min=0.12
    Lifetime Mode=::NewEnumerator1
    Mesh Scale Max=<1.1 0.7 0.7>
    Mesh Scale Min=<0.65 0.45 0.45>
    Mesh Scale Mode=::NewEnumerator2
    UsePositionOffset=false
    Write Material Random=true
   )
   &AddVelocity(Velocity=&RandomRangeVector(Minimum=<-30 -15 15>, Maximum=<-10 15 40>, Evaluation Type=::NewEnumerator1))
   &AddVelocity(Velocity=&Multiply_VectorByFloat(
     Vector=&RandomRangeVector(Minimum=<0 10 0>, Maximum=<0 10 0>, Evaluation Type=::NewEnumerator1)
     Float=&Lerp_Float(
      A=-1
      B=1
      Alpha=&MakeFloatFromInt(Int=/Niagara/DynamicInputs/Subtract/Subtract_Integer(A:NiagaraInt32=&SpawnGroup(), B:NiagaraInt32=2))
     )
    )) {module=AddVelocity001}
   &DynamicMaterialParameters(
    Index 0 Param 3=&RandomRangeFloat(Minimum=0.75, Maximum=1, Evaluation Type=::NewEnumerator1)
    Module.Param0WriteEnabled=false
    Module.Param1WriteEnabled=false
    Module.Param2WriteEnabled=true
    Module.Param3WriteEnabled=false
   ) {module=DynamicMaterialParameters001}
  ]
  particleUpdateScript [
   &ParticleState()
   &SkeletalMeshLocation(
    Bone / Socket Index=&SpawnGroup()
    "Inherit Velocity (Scale)"=-0.1
    "Interpolate Mesh Sampling (Update)"=true
    MaskSpawnGroup=false
    Mesh Sampling Type=::NewEnumerator5
    Orientation Sampling=::NewEnumerator1
    Position Sampling=::NewEnumerator1
    Skeletal Mesh=$User.SkeletalMesh
    Socket Sampling Mode=::NewEnumerator5
    Velocity Sampling=::NewEnumerator0
    VelocitySpeedLimit=false
   )
   &Drag(Drag=4, Use Linear Drag=true, Use Rotational Drag=false)
   &SolveForcesAndVelocity(Clamp Velocity=false, Limit Acceleration=false)
   &ScaleMeshSize(Scale Factor=&VectorFromCurve(
     VectorCurve=DataInterface() {propertyValues {
       XCurve {keys [
         {arriveTangent=2.2467, interpMode=RCIM_Cubic, time=0, value=0}
         {arriveTangent=1, interpMode=RCIM_Cubic, leaveTangent=1, time=0.4977, value=0.7381}
         {interpMode=RCIM_Cubic, leaveTangent=0.2845, time=1, value=1}
        ]}
       YCurve {keys [<at 0 0.2 RCIM_Cubic>, <at 0.4977 1 RCIM_Cubic>, {interpMode=RCIM_Cubic, leaveTangent=-4.9675, time=1, value=0.2}]}
       ZCurve {keys [<at 0 0.2 RCIM_Cubic>, <at 0.4977 1 RCIM_Cubic>, {interpMode=RCIM_Cubic, leaveTangent=-5.0464, time=1, value=0.2}]}
      }}
     Scale Curve=&Lerp_Vector(
      A=<1 1 1>
      B=<2.5 1.25 1.25>
      Alpha=&Lerp_Float(
       A=$User.LeftVFast
       B=$User.RightVFast
       Alpha=&MakeFloatFromInt(Int=&Subtract_Integer(A:NiagaraInt32=&SpawnGroup(), B:NiagaraInt32=2))
      )
     )
    ))
   &ScaleColor(
    Scale Alpha=&FloatFromCurve(
     FloatCurve=DataInterface() {propertyValues {Curve {keys [<at 0 1 RCIM_Cubic>, {arriveTangent=-2.4164, interpMode=RCIM_Cubic, leaveTangent=-2.4164, tangentMode=RCTM_User, time=1, value=0}]}}}
     Scale Curve=&RandomRangeFloat(Minimum=0.4, Maximum=0.9)
    )
    Scale RGB=&VectorFromCurve(VectorCurve=DataInterface() {propertyValues {XCurve {keys [<at 0 1>,<at 1 1>]}, YCurve {keys [<at 0 1>,<at 1 1>]}, ZCurve {keys [<at 0 1>,<at 1 1>]}}})
    ScaleA=true
    ScaleRGB=true
   )
   &DynamicMaterialParameters(
    Index 0 Param 1=&FloatFromCurve(FloatCurve=DataInterface() {propertyValues {Curve {keys [<at 0.1 0.1 RCIM_Cubic>,<at 0.8 1>]}}})
    Index 0 Param 2=2
    Index 0 Param 4=&FloatFromCurve(FloatCurve=DataInterface() {propertyValues {Curve {keys [<at 0 1>,<at 1 3>]}}})
    Module.Param0WriteEnabled=true
    Module.Param1WriteEnabled=true
    Module.Param2WriteEnabled=false
    Module.Param3WriteEnabled=true
   )
   &CameraOffset(Camera Offset Amount=2)
  ]
  rendererClasses [&NiagaraMeshRendererProperties]
  renderers [{rendererClass=&NiagaraMeshRendererProperties, rendererIndex=0}]
 }
 Smoke {
  emitterSpawnScript [/SetVariables(Emitter.Depth Fade High=10, Emitter.Depth Fade Low=30) {isSetParameters=true}]
  emitterUpdateScript [
   &EmitterState(Life Cycle Mode=::NewEnumerator1, Loop Duration=1, UseLoopDelay=true)
   &SpawnRate(
    SpawnRate=&Multiply_Float(
     A=&Lerp_Float(A=0, B=30, Alpha=$User.LeftSlow)
     B=&Lerp_Float(A=1, B=2, Alpha=$User.LeftFast)
    )
    Use Spawn Probability=false
   ) {module=SpawnRate002}
   &SpawnRate(
    SpawnGroup=1
    SpawnRate=&Multiply_Float(
     A=&Lerp_Float(A=0, B=30, Alpha=$User.RightSlow)
     B=&Lerp_Float(A=1, B=2, Alpha=$User.RightFast)
    )
    Use Spawn Probability=false
   ) {module=SpawnRate003}
   &SpawnRate(
    SpawnGroup=2
    SpawnRate=&Multiply_Float(
     A=&Lerp_Float(A=0, B=30, Alpha=$User.LeftSlow)
     B=&Lerp_Float(A=1, B=2, Alpha=$User.LeftFast)
    )
    Use Spawn Probability=false
   )
   &SpawnRate(
    SpawnGroup=3
    SpawnRate=&Multiply_Float(
     A=&Lerp_Float(A=0, B=30, Alpha=$User.RightSlow)
     B=&Lerp_Float(A=1, B=2, Alpha=$User.RightFast)
    )
    Use Spawn Probability=false
   ) {module=SpawnRate001}
  ]
  particleSpawnScript [
   &InitializeParticle(
    Color=<rgba 0.6562 0.6155 0.5913 1>
    Lifetime Max=0.4
    Lifetime Min=0.25
    Lifetime Mode=::NewEnumerator1
    Sprite Rotation Mode=::NewEnumerator2
    Sprite Size Mode=::NewEnumerator1
    Uniform Sprite Size Max=45
    Uniform Sprite Size Min=25
    UsePositionOffset=false
    Write Material Random=true
   )
   &AddVelocity(
    Scale Added Velocity=&SelectVectorFromArray(
     Direct Array Index=&SpawnGroup()
     Vector Selection Array=DataInterface() {propertyValues {FloatData [<1 -1 1>, <1 1 1>, <1 -1 1>, <1 1 1>]}}
    )
    Velocity=&RandomRangeVector(Minimum=<0 80 70>, Maximum=<0 180 150>, Evaluation Type=::NewEnumerator1)
   )
   &SkeletalMeshLocation(
    Bone / Socket Index=&SpawnGroup()
    MaskSpawnGroup=false
    Mesh Sampling Type=::NewEnumerator5
    Skeletal Mesh=$User.SkeletalMesh
    Socket Sampling Mode=::NewEnumerator5
   )
  ]
  particleUpdateScript [
   &ParticleState()
   &Drag(Drag=10, Use Linear Drag=true, Use Rotational Drag=false)
   &ScaleColor(
    Scale Alpha=&FloatFromCurve(
     FloatCurve=DataInterface() {propertyValues {Curve {keys [
         {arriveTangent=20.3455, interpMode=RCIM_Cubic, leaveTangent=20.3455, tangentMode=RCTM_User, time=0, value=0}
         {arriveTangent=-2.0712, interpMode=RCIM_Cubic, leaveTangent=-2.0712, tangentMode=RCTM_User, time=0.1, value=0.9847}
         {arriveTangent=-0.0405, interpMode=RCIM_Cubic, leaveTangent=-0.0405, tangentMode=RCTM_User, time=1, value=0}
        ]}}}
     Scale Curve=&RandomRangeFloat(Minimum=0.25, Maximum=0.35)
    )
    Scale RGB=<0.8 0.8 0.8>
    ScaleA=true
    ScaleRGB=true
   )
   &SolveForcesAndVelocity(Clamp Velocity=false, Limit Acceleration=false)
   &ScaleSpriteSizeBySpeed(Max Scale Factor=<1 1.5>, Source Velocity:Vector3f=&Multiply_Vector(A=<0 0.01 1>, B=$Particles.Velocity))
   &ScaleSpriteSize(Uniform Curve Sprite Scale=DataInterface() {propertyValues {Curve {keys [
        {arriveTangent=2.1875, interpMode=RCIM_Cubic, leaveTangent=2.1875, tangentMode=RCTM_User, time=0, value=0.2}
        {arriveTangent=-0.0405, interpMode=RCIM_Cubic, leaveTangent=-0.0405, tangentMode=RCTM_User, time=1, value=1}
       ]}}})
   &CameraOffset(Camera Offset Amount=16)
   &DynamicMaterialParameters(
    Index 0 Param 1=&Add_Float(
     A=0
     B=&FloatFromCurve(FloatCurve=DataInterface() {propertyValues {Curve {keys [{tangentMode=RCTM_User, time=0, value=0}, {tangentMode=RCTM_User, time=1, value=1}]}}})
    )
    Index 0 Param 3=&FloatFromCurve(FloatCurve=DataInterface() {propertyValues {Curve {keys [{arriveTangent=-53.9199, time=0, value=0}, {leaveTangent=0.2723, time=0.25, value=0}]}}})
    Index 0 Param 4=&FloatFromCurve(FloatCurve=DataInterface() {propertyValues {Curve {keys [{arriveTangent=-3.3947, interpMode=RCIM_Cubic, leaveTangent=-3.3947, tangentMode=RCTM_User, time=0, value=1}, <at 1 0>]}}})
    Index 1 Param 1=&FloatFromCurve(FloatCurve=DataInterface() {propertyValues {Curve {keys [<at 0 0>, {arriveTangent=4.2257, interpMode=RCIM_Cubic, leaveTangent=4.2257, tangentMode=RCTM_User, time=0.25, value=0}]}}})
    Index 1 Param 2=&FloatFromCurve(FloatCurve=DataInterface() {propertyValues {Curve {keys [<at 0 0>]}}})
    Index 1 Param 3=&FloatFromCurve(FloatCurve=DataInterface() {propertyValues {Curve {keys [{interpMode=RCIM_Cubic, leaveTangent=6.3917, tangentMode=RCTM_Break, time=0, value=0}]}}})
    Index 1 Param 4=&FloatFromCurve(FloatCurve=DataInterface() {propertyValues {Curve {keys [{arriveTangent=-0.136, leaveTangent=-0.136, time=0, value=1}]}}})
    Index 2 Param 1=&FloatFromCurve(FloatCurve=DataInterface() {propertyValues {Curve {keys [
         {arriveTangent=5.4363, interpMode=RCIM_Cubic, time=0.1, value=0.5}
         {arriveTangent=1.5446, interpMode=RCIM_Cubic, leaveTangent=1.5446, tangentMode=RCTM_User, time=0.5, value=1}
        ]}}})
    Index 2 Param 2=&FloatFromCurve(
     FloatCurve=DataInterface() {propertyValues {Curve {keys [
         {arriveTangent=1.9056, time=0, value=1}
         {arriveTangent=0.0259, interpMode=RCIM_Cubic, leaveTangent=0.0259, tangentMode=RCTM_User, time=1, value=1.2}
        ]}}}
     Scale Curve=&RandomRangeFloat(Minimum=0.1, Maximum=0.2)
    )
    Index 2 Param 3=&FloatFromCurve(FloatCurve=DataInterface() {propertyValues {Curve {keys [
         {interpMode=RCIM_Cubic, leaveTangent=0.7306, tangentMode=RCTM_Break, time=0, value=1}
         {interpMode=RCIM_Cubic, leaveTangent=11.1233, time=1, value=1.5}
        ]}}})
    Index 2 Param 4=&FloatFromCurve(FloatCurve=DataInterface() {propertyValues {Curve {keys [
         {arriveTangent=-0.6867, interpMode=RCIM_Cubic, time=0, value=1}
         {interpMode=RCIM_Cubic, leaveTangent=-1.2956, tangentMode=RCTM_Break, time=1, value=0.8}
        ]}}})
    Module.Param0WriteEnabled=true
    Module.Param0WriteEnabled1=true
    Module.Param0WriteEnabled2=true
    Module.Param1WriteEnabled=false
    Module.Param1WriteEnabled1=true
    Module.Param1WriteEnabled2=true
    Module.Param2WriteEnabled=true
    Module.Param2WriteEnabled1=true
    Module.Param2WriteEnabled2=true
    Module.Param3WriteEnabled=true
    Module.Param3WriteEnabled1=true
    Module.Param3WriteEnabled2=true
    Write Parameter Index 1=true
    Write Parameter Index 2=true
   )
  ]
  rendererClasses [&NiagaraSpriteRendererProperties]
  renderers [
   {rendererClass=&NiagaraSpriteRendererProperties}
   {rendererClass=&NiagaraSpriteRendererProperties, rendererIndex=1}
  ]
 }
 SparksLocal {
  simTarget=GPUComputeSim
  emitterUpdateScript [
   &EmitterState(Life Cycle Mode=::NewEnumerator1, Loop Duration=1, UseLoopDelay=false)
   &SpawnRate(SpawnRate=&Lerp_Float(A=0, B=100, Alpha=$User.LeftVFast), Use Spawn Probability=false) {module=SpawnRate002}
   &SpawnRate(SpawnGroup=1, SpawnRate=&Lerp_Float(A=0, B=100, Alpha=$User.RightVFast), Use Spawn Probability=false) {module=SpawnRate003}
   &SpawnRate(SpawnGroup=2, SpawnRate=&Lerp_Float(A=0, B=100, Alpha=$User.LeftVFast), Use Spawn Probability=false)
   &SpawnRate(SpawnGroup=3, SpawnRate=&Lerp_Float(A=0, B=100, Alpha=$User.RightVFast), Use Spawn Probability=false) {module=SpawnRate001}
  ]
  particleSpawnScript [
   &InitializeParticle(
    Lifetime Max=0.11
    Lifetime Min=0.07
    Lifetime Mode=::NewEnumerator1
    Sprite Size Max=<1 30>
    Sprite Size Min=<0.7 8>
    Sprite Size Mode=::NewEnumerator2
    UsePositionOffset=false
    Write Material Random=true
   )
   &SphereLocation(
    Hemisphere Z=true
    MaskSpawnGroup=false
    Offset=<-12 0 0.5>
    Override Seed=false
    Sphere Radius=&RandomRangeFloat(Minimum=2, Maximum=3)
    Surface Only=true
    UseNonUniformScale=true
    UseOffset=true
    UseSphereOrientation=false
   )
   &AddVelocity(Velocity=&Multiply_VectorByFloat(
     Vector=$Output.SphereLocation.SphereVector
     Float=&RandomRangeFloat(Minimum=20, Maximum=50, Evaluation Type=::NewEnumerator1)
    ))
   &Color(Color=&MakeLinearColorFromVectorAndFloat("Vector (RGB)"=<2 0.9881 0.248>, "Float (Alpha)"=&ClampFloat(Float=1)))
   &AddVelocity(
    Scale Added Velocity=&SelectVectorFromArray(
     Direct Array Index=&SpawnGroup()
     Vector Selection Array=DataInterface() {propertyValues {FloatData [<1 -1 1>, <1 1 1>, <1 -1 1>, <1 1 1>]}}
    )
    Velocity=&RandomRangeVector(Minimum=<0 0 50>, Maximum=<0 200 150>, Evaluation Type=::NewEnumerator1)
   ) {module=AddVelocity001}
  ]
  particleUpdateScript [
   &ParticleState()
   &SkeletalMeshLocation(
    Bone / Socket Index=&SpawnGroup()
    "Inherit Velocity (Scale)"=-0.5
    MaskSpawnGroup=false
    Mesh Sampling Type=::NewEnumerator5
    Sampled Position Offset=<0 0 -6>
    Skeletal Mesh=$User.SkeletalMesh
    Socket Sampling Mode=::NewEnumerator5
    Velocity Sampling=::NewEnumerator0
    VelocitySpeedLimit=false
   )
   &SolveForcesAndVelocity(Clamp Velocity=false, Limit Acceleration=false)
   &ScaleSpriteSize(
    Scale Factor=&Vector2DFromCurve(
     Vector2Curve=DataInterface() {propertyValues {
       XCurve {keys [<at 0 1 RCIM_Cubic>, {arriveTangent=-1, interpMode=RCIM_Cubic, leaveTangent=-1, time=0.07, value=1}, <at 1 0 RCIM_Cubic>]}
       YCurve {keys [{arriveTangent=43.0597, interpMode=RCIM_Cubic, time=0, value=0}, <at 0.07 1 RCIM_Cubic>, <at 1 0 RCIM_Cubic>]}
      }}
     ScaleCurve=<1 1>
    )
    Scale Sprite Size Mode=::NewEnumerator2
   )
   &ScaleColor(
    Scale Alpha=&FloatFromCurve(FloatCurve=DataInterface() {propertyValues {Curve {keys [<at 0 2 RCIM_Cubic>,<at 0.2 1>]}}})
    Scale RGB=&VectorFromCurve(VectorCurve=DataInterface() {propertyValues {
       XCurve {keys [<at 0.2 64 RCIM_Cubic>,<at 1 1 RCIM_Cubic>]}
       YCurve {keys [<at 0.2 64 RCIM_Cubic>,<at 1 0.1249 RCIM_Cubic>]}
       ZCurve {keys [<at 0.2 64 RCIM_Cubic>,<at 1 0.2547 RCIM_Cubic>]}
      }})
    ScaleA=true
    ScaleRGB=true
   )
  ]
  rendererClasses [&NiagaraSpriteRendererProperties]
  renderers [{rendererClass=&NiagaraSpriteRendererProperties}]
 }
 SparksSpray {
  simTarget=GPUComputeSim
  emitterUpdateScript [
   &EmitterState(
    Life Cycle Mode=::NewEnumerator1
    Loop Duration=&RandomRangeFloat(Minimum=0.1, Maximum=0.3)
    Recalculate Duration Each Loop=true
    UseLoopDelay=true
   )
   &SpawnBurst_Instantaneous(
    Spawn Count=&RandomRangeInt(Minimum=16, Maximum=25)
    Spawn Probability=$User.LeftVFast
    Use Spawn Probability=true
    UseLoopCountLimit=false
   ) {module=SpawnBurst_Instantaneous002}
   &SpawnBurst_Instantaneous(
    Spawn Count=&RandomRangeInt(Minimum=16, Maximum=25)
    Spawn Group=1
    Spawn Probability=$User.RightVFast
    Use Spawn Probability=true
    UseLoopCountLimit=false
   ) {module=SpawnBurst_Instantaneous003}
   &SpawnBurst_Instantaneous(
    Spawn Count=&RandomRangeInt(Minimum=16, Maximum=25)
    Spawn Group=2
    Spawn Probability=$User.LeftVFast
    Use Spawn Probability=true
    UseLoopCountLimit=false
   )
   &SpawnBurst_Instantaneous(
    Spawn Count=&RandomRangeInt(Minimum=16, Maximum=25)
    Spawn Group=3
    Spawn Probability=$User.RightVFast
    Use Spawn Probability=true
    UseLoopCountLimit=false
   ) {module=SpawnBurst_Instantaneous001}
  ]
  particleSpawnScript [
   &InitializeParticle(
    Lifetime Max=0.3
    Lifetime Min=0.08
    Lifetime Mode=::NewEnumerator1
    Sprite Size Max=<1.25 3>
    Sprite Size Min=<0.75 1>
    Sprite Size Mode=::NewEnumerator2
    UsePositionOffset=false
    Write Material Random=true
   )
   &SkeletalMeshLocation(
    Bone / Socket Index=&SpawnGroup()
    MaskSpawnGroup=false
    Mesh Sampling Type=::NewEnumerator5
    Sampled Position Offset=<0 0 -6>
    Skeletal Mesh=$User.SkeletalMesh
    Socket Sampling Mode=::NewEnumerator5
   )
   &SphereLocation(
    Hemisphere X=true
    Hemisphere Z=true
    MaskSpawnGroup=false
    Non Uniform Scale=<-1 1 1>
    Override Seed=false
    Sphere Radius=&RandomRangeFloat(Minimum=2, Maximum=3)
    Surface Only=true
    UseNonUniformScale=true
    UseOffset=false
    UseSphereOrientation=false
   )
   &AddVelocity(Velocity=&Multiply_VectorByFloat(
     Vector=$Output.SphereLocation.SphereVector
     Float=&RandomRangeFloat(Minimum=20, Maximum=60, Evaluation Type=::NewEnumerator1)
    ))
   &Color(Color=&MakeLinearColorFromVectorAndFloat("Vector (RGB)"=<2 0.9881 0.248>, "Float (Alpha)"=&ClampFloat(Float=1)))
   &AddVelocity(
    Scale Added Velocity=&SelectVectorFromArray(
     Direct Array Index=&SpawnGroup()
     Vector Selection Array=DataInterface() {propertyValues {FloatData [<1 -1 1>, <1 1 1>, <1 -1 1>, <1 1 1>]}}
    )
    Velocity=&RandomRangeVector(Minimum=<0 0 50>, Maximum=<0 200 150>, Evaluation Type=::NewEnumerator1)
   ) {module=AddVelocity001}
   &InheritVelocity(Inherited Velocity Amount Scale=<-0.5 -0.5 -0.25>, Inherited Velocity Speed Limit=-1000, Speed Limit=true, UseSpeedThreshold=false)
  ]
  particleUpdateScript [
   &ParticleState()
   &AccelerationForce(Acceleration=<0 0 -800>)
   &SolveForcesAndVelocity(Clamp Velocity=false, Limit Acceleration=false)
   &ScaleSpriteSize(
    Scale Factor=&Vector2DFromFloat(Value=&FloatFromCurve(FloatCurve=DataInterface() {propertyValues {Curve {keys [
          {arriveTangent=55.4408, interpMode=RCIM_Cubic, time=0, value=0}
          <at 0.03 1 RCIM_Cubic>
          {arriveTangent=-3.3237, leaveTangent=-3.3237, time=1, value=0}
         ]}}}))
    Scale Sprite Size Mode=::NewEnumerator2
   )
   &ScaleSpriteSizeBySpeed(
    Initial Sprite Size:Vector2f=$Particles.SpriteSize
    Max Scale Factor=<0.5 100>
    Source Velocity:Vector3f=&Multiply_Vector(A=<0.01 0.01 0.2>, B=$Particles.Velocity)
   )
   &ScaleColor(
    Scale Alpha=&FloatFromCurve(FloatCurve=DataInterface() {propertyValues {Curve {keys [<at 0 2 RCIM_Cubic>,<at 0.2 1>]}}})
    Scale RGB=&VectorFromCurve(VectorCurve=DataInterface() {propertyValues {
       XCurve {keys [<at 0.2 64 RCIM_Cubic>,<at 1 1 RCIM_Cubic>]}
       YCurve {keys [<at 0.2 64 RCIM_Cubic>,<at 1 0.1249 RCIM_Cubic>]}
       ZCurve {keys [<at 0.2 64 RCIM_Cubic>,<at 1 0.2547 RCIM_Cubic>]}
      }})
    ScaleA=true
    ScaleRGB=true
   )
  ]
  rendererClasses [&NiagaraSpriteRendererProperties]
  renderers [{rendererClass=&NiagaraSpriteRendererProperties}]
 }
 SpeedLines {
  emitterUpdateScript [
   &EmitterState()
   &SpawnRate(
    SpawnRate=&FloatFromCurve(
     FloatCurve=DataInterface() {propertyValues {Curve {keys [
         {arriveTangent=0.4214, interpMode=RCIM_Cubic, leaveTangent=0.4225, tangentMode=RCTM_User, time=0, value=0}
         <at 0.3926 171.1499>
         {arriveTangent=16.7131, interpMode=RCIM_Cubic, leaveTangent=16.7131, tangentMode=RCTM_User, time=0.7152, value=203.2333}
         <at 0.9 350>
        ]}}}
     CurveIndex=$User.Fast
    )
    Use Spawn Probability=false
   )
  ]
  particleSpawnScript [
   &InitializeParticle(
    Lifetime Max=0.5
    Lifetime Min=0.4
    Lifetime Mode=::NewEnumerator1
    Sprite Rotation Mode=::NewEnumerator2
    Sprite Size=<0.065 1>
    Sprite Size Mode=::NewEnumerator4
    UsePositionOffset=false
    Write Material Random=true
   )
   &SkeletalMeshLocation(
    Bone / Socket Index=4
    MaskSpawnGroup=false
    Mesh Sampling Type=::NewEnumerator5
    Skeletal Mesh=$User.SkeletalMesh
    Socket Sampling Mode=::NewEnumerator5
   )
   &TorusLocation(
    Align Torus With Axis=true
    Handle Radius=100
    Large Radius=200
    MaskSpawnGroup=false
    Module.Override Randomness=false
    Offset=&Multiply_VectorByFloat(Vector=$Engine.Owner.SystemXAxis, Float=450)
    Torus Axis=&RotateVectorByQuaternion(
     Vector To Rotate=<1 0 0>
     Quaternion=&MakeQuatFromAxisAngle(Axis=$Engine.Owner.SystemXAxis, AngleInDegrees=0)
    )
   )
   &AddVelocity(Velocity=&RotateVectorByQuaternion(
     Vector To Rotate=<-2500 0 0>
     Quaternion=&MakeQuatFromAxisAngle(Axis=$Engine.Owner.SystemXAxis, AngleInDegrees=0)
    ))
  ]
  particleUpdateScript [
   &ParticleState()
   &Color(
    Color=&ScaleLinearColorByCurve(
     InitialValue=<rgba 1 1 1 1>
     ScaleCurve=DataInterface() {propertyValues {
       AlphaCurve {keys [<at 0.0384 1>,<at 0.3631 1>]}
       BlueCurve {keys [<at 0.0281 1>]}
       GreenCurve {keys [<at 0.0281 1>]}
       RedCurve {keys [<at 0.0281 1>]}
      }}
     CurveIndex=$Particles.NormalizedAge
    )
    Scale Alpha=&FloatFromCurve(FloatCurve=DataInterface() {propertyValues {Curve {keys [<at 0 0 RCIM_Cubic>,<at 0.5 1 RCIM_Cubic>,<at 1 0 RCIM_Cubic>]}}})
   )
   &ScaleColor(
    Scale Alpha=&FloatFromCurve(
     FloatCurve=DataInterface() {propertyValues {Curve {keys [{arriveTangent=1.7024, interpMode=RCIM_Cubic, leaveTangent=1.7024, tangentMode=RCTM_User, time=0, value=0}, <at 1 1>]}}}
     CurveIndex=$User.Fast
    )
    ScaleA=true
    ScaleRGB=true
   )
   &ScaleColor(
    Scale Alpha=&FloatFromCurve(FloatCurve=DataInterface() {propertyValues {Curve {keys [<at 0 1>,<at 1 0>]}}})
    ScaleA=true
    ScaleRGB=true
   ) {module=ScaleColor001}
   &SolveForcesAndVelocity(Clamp Velocity=false, Limit Acceleration=false)
   &ScaleSpriteSizeBySpeed(Max Scale Factor=<0.5 1>, Min Scale Factor=<0 0>, Velocity Threshold=4250)
   &ScaleSpriteSize(Scale Factor=&Vector2DFromCurve(
     Vector2Curve=DataInterface() {propertyValues {
       XCurve {keys [<at -0.0069 0.5 RCIM_Cubic>,<at 0.4918 1 RCIM_Cubic>,<at 1 0.5 RCIM_Cubic>]}
       YCurve {keys [<at 0 0.0245 RCIM_Cubic>, {arriveTangent=-0.0245, interpMode=RCIM_Cubic, leaveTangent=-0.0245, time=0.4975, value=1}, <at 1 0 RCIM_Cubic>]}
      }}
     ScaleCurve=&Multiply_Vector2D_ByFloat(
      Float=&FloatFromCurve(
       FloatCurve=DataInterface() {propertyValues {Curve {keys [
           {arriveTangent=0.4214, interpMode=RCIM_Cubic, leaveTangent=0.4225, tangentMode=RCTM_User, time=0, value=0}
           {arriveTangent=7.1685, interpMode=RCIM_Cubic, leaveTangent=7.1685, tangentMode=RCTM_User, time=0.167, value=127.7715}
           {arriveTangent=7.5, interpMode=RCIM_Cubic, leaveTangent=7.5, tangentMode=RCTM_User, time=0.7035, value=130.4144}
           {arriveTangent=1328.8317, interpMode=RCIM_Cubic, leaveTangent=1328.8324, tangentMode=RCTM_User, time=1, value=288.3149}
          ]}}}
       CurveIndex=$User.Fast
      )
      Vector2D=<1 1>
     )
    ))
   &CameraOffset(Camera Offset Amount=-50)
  ]
  rendererClasses [&NiagaraSpriteRendererProperties]
  renderers [{rendererClass=&NiagaraSpriteRendererProperties}]
 }
}
summary {
 systemName=NS_Shopping_Cart_Vehicle_WheelDust
 emitters [
  {bEnabled=false, emitterName=FN_SpriteFromSkeleton_Sparks, rendererClasses [&NiagaraSpriteRendererProperties]}
  {bEnabled=false, emitterName=FN_SpriteFromSkeleton_SparksLocal, rendererClasses [&NiagaraSpriteRendererProperties]}
  {bEnabled=false, emitterName=FN_SpriteFromSkeleton_Smoke, rendererClasses [&NiagaraSpriteRendererProperties]}
  {bEnabled=false, emitterName=FN_SpriteFromSkeleton_Debirs, rendererClasses [&NiagaraSpriteRendererProperties]}
  {bEnabled=false, emitterName=FN_MeshFromSkeleton_Sharp, rendererClasses [&NiagaraMeshRendererProperties]}
  {bEnabled=false, emitterName=FN_Emitter, rendererClasses [&NiagaraSpriteRendererProperties]}
  {bEnabled=false, emitterName=FN_Sprite_SpeedLines, rendererClasses [&NiagaraSpriteRendererProperties]}
  {bEnabled=false, emitterName=FN_Sprite_SpeedLinesSlow, rendererClasses [&NiagaraSpriteRendererProperties]}
  {emitterName=Smoke, rendererClasses [&NiagaraSpriteRendererProperties]}
  {emitterName=Sharp, rendererClasses [&NiagaraMeshRendererProperties]}
  {bEnabled=false, emitterName=LeftChunks, rendererClasses [&NiagaraSpriteRendererProperties]}
  {emitterName=SparksLocal, simTarget=GPUComputeSim, rendererClasses [&NiagaraSpriteRendererProperties]}
  {emitterName=SparksSpray, simTarget=GPUComputeSim, rendererClasses [&NiagaraSpriteRendererProperties]}
  {emitterName=SpeedLines, rendererClasses [&NiagaraSpriteRendererProperties]}
 ]
 userVariables [
  User.Fast {type=NiagaraFloat, defaultValue=0, description=""}
  User.LeftFast {type=NiagaraFloat, defaultValue=0, description=""}
  User.LeftSlow {type=NiagaraFloat, defaultValue=0, description=""}
  User.LeftVFast {type=NiagaraFloat, defaultValue=0, description=""}
  User.RightFast {type=NiagaraFloat, defaultValue=0, description=""}
  User.RightSlow {type=NiagaraFloat, defaultValue=0, description=""}
  User.RightVFast {type=NiagaraFloat, defaultValue=0, description=""}
  User.SkeletalMesh {
   type=NiagaraDataInterfaceSkeletalMesh
   defaultValue {
    struct=NiagaraExt_VariableValue_DataInterface
    value {dataInterface=&self:NiagaraDataInterfaceSkeletalMesh_2}
   }
  }
 ]
}
